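; Ice King countdown actor with a twelve-orb clock face and a channeling child-actor mannequin.
; The construction script rings the orbs, sets ice bubble transparency and removes the prop checker if disabled; otherwise the checker destroys overlapping building props.
; BeginPlay attaches channeling effects to the mannequin's left hand and optionally starts a client-only timer yawing the clock toward the camera.
; Custom events drive frost cover visibility and lerp channeling particle parameters by hour.
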(event Custom|Face%20Camera
  (bind _returnvalue (Fortnite|GetLocalFortPlayerControllers true true))
  (bind _output (Utilities|Array|Get(acopy) _returnvalue))
  (bind _clockface (Variables|Default|GetClockFace))
  (bind _playercameramanager (Variables|/Script/Engine.PlayerController|GetPlayerCameraManager _output))
  ;[1 1 1] Rotate the clock face to always face the camera
  (Utilities|IsValid _output
    (:Is%20Valid
      (Utilities|IsValid _playercameramanager
        (:Is%20Valid
          (Utilities|IsValid _clockface
            (:Is%20Valid
              (Transformation|SetWorldRotation _clockface (Math|Rotator|MakeRotator 0.0 0.0 (.yaw (Math|Rotator|FindLookatRotation (Camera|GetCameraLocation _playercameramanager) (Transformation|GetActorLocation)))) false false))))))))

(event EventBeginPlay
  (bind _assoldier_mannequin_male_child_mooney_channeling (Utilities|Casting|CastToSoldierMannequin_Male_Child_Mooney_Channeling (Variables|/Script/Engine.ChildActorComponent|GetChildActor (Variables|Default|GetIce%20King%20-%20ChildActor))))
  (Transformation|AttachComponentToComponent (Variables|Default|GetP_Mooney_Channeling) (Variables|/Script/Engine.SkeletalMeshActor|GetSkeletalMeshComponent _assoldier_mannequin_male_child_mooney_channeling) "LeftHand" "SnapToTarget" "KeepRelative" "KeepRelative" true)
  (Transformation|AttachComponentToComponent (Variables|Default|GetP_Mooney_Channeling_Outro) (Variables|/Script/Engine.SkeletalMeshActor|GetSkeletalMeshComponent _assoldier_mannequin_male_child_mooney_channeling) "LeftHand" "SnapToTarget" "KeepRelative" "KeepRelative" true)
  (if (Variables|Default|GetShouldFaceCamera)
    (Utilities|FlowControl|Delay 5.0)
    (Networking|SwitchHasAuthority
      (:Remote
        (Utilities|Time|SetTimerbyEvent (EventDispatchers|CreateEvent :!event Custom|Face%20Camera) (Variables|Default|GetRotationTime) true)))))

(event Custom|KingVisible
  (Rendering|Material|SetScalarParameterValueonMaterials (Variables|Default|GetFrostScreenCover) "CountdownStarted" (Math|Conversions|ToFloat(Boolean) (Variables|Default|GetIce%20King%20Visible)))
  (Components|Activation|Activate (Variables|Default|GetP_Mooney_Channeling) false))

(event Custom|Hourly%20Change
  (params (Hour :type Integer))
  (bind _returnvalue (/ Hour 2))
  (bind _returnvalue_2 (Math|Conversions|ToFloat(Integer) _returnvalue))
  (bind _returnvalue_3 (/ _returnvalue_2 12.0))
  (bind _returnvalue_4 (* _returnvalue_3 2.0))
  (bind _p_mooney_channeling (Variables|Default|GetP_Mooney_Channeling))
  (Effects|Components|ParticleSystem|SetFloatParameter _p_mooney_channeling "SnowRate" (Math|Float|Lerp (Variables|Default|GetSnowRate_Max) (Variables|Default|GetSnowRate) _returnvalue_3))
  (Effects|Components|ParticleSystem|SetVectorParameter _p_mooney_channeling "LightSize" (Math|Vector|Lerp(Vector) (Variables|Default|GetLight%20Size_Max) (Variables|Default|GetLight%20Size) _returnvalue_3))
  (Effects|Components|ParticleSystem|SetFloatParameter _p_mooney_channeling "CoronaAlpha" (Math|Float|Lerp (Variables|Default|GetCorona%20Alpha_Max) (Variables|Default|GetCorona%20Alpha) _returnvalue_3))
  (Effects|Components|ParticleSystem|SetVectorParameter _p_mooney_channeling "FlareSize" (Math|Vector|Lerp(Vector) (Variables|Default|GetFlare%20Size_Max) (Variables|Default|GetFlare%20Size) _returnvalue_3))
  (Effects|Components|ParticleSystem|SetVectorParameter _p_mooney_channeling "CubeSize" (Math|Vector|Lerp(Vector) (Variables|Default|GetCube%20Size_Max) (Variables|Default|GetCube%20Size) _returnvalue_3))
  (Effects|Components|ParticleSystem|SetVectorParameter _p_mooney_channeling "AniFlareSize" (Math|Vector|Lerp(Vector) (Variables|Default|GetAni_Flare%20Size_Max) (Variables|Default|GetAni_Flare%20Size) _returnvalue_3))
  (if (>= _returnvalue_4 0.5)
    (Effects|Components|ParticleSystem|SetFloatParameter _p_mooney_channeling "ElecSpawn" (Math|Float|Lerp (Variables|Default|GetElec%20Spawn_Max) (Variables|Default|GetElec%20Spawn) _returnvalue_4))))

(event OnComponentBeginOverlap(BuildingPropChecker)
  (params (OverlappedComponent :type Primitive%20Component%20Object%20Reference) (OtherActor :type Actor%20Object%20Reference) (OtherComp :type Primitive%20Component%20Object%20Reference) (OtherBodyIndex :type Integer) (bFromSweep :type Boolean) (SweepResult :type Hit%20Result%20Structure%20%28by%20ref%29))
  (bind _asbuilding_prop (Utilities|Casting|CastToBuildingProp OtherActor))
  (Actor|DestroyActor _asbuilding_prop))

(fn ConstructionScript
  (Utilities|FlowControl|Sequence
    (:then_0
      ;[1 1 1] Setup Clock Face Particle Systems
      (LocalVariables|SetAngle (/ (Math|Trig|GetPI) 2.0))
      (bind _output_get (Variables|Default|SetInner_TimeOrbs (Utilities|Array|MakeArray (Variables|Default|Get0) (Variables|Default|Get1) (Variables|Default|Get2) (Variables|Default|Get3) (Variables|Default|Get4) (Variables|Default|Get5) (Variables|Default|Get6) (Variables|Default|Get7) (Variables|Default|Get8) (Variables|Default|Get9) (Variables|Default|Get10) (Variables|Default|Get11))))
      (for (_array_element _array_index) _output_get
        (Transformation|SetRelativeLocation _array_element (Math|Vector|MakeVector 0.0 (* (Variables|/Script/Engine.SphereComponent|GetSphereRadius (Variables|Default|GetInner%20Ring%20Radius%20Control)) (Math|Trig|Cos(Radians) (LocalVariables|GetAngle))) (* (Variables|/Script/Engine.SphereComponent|GetSphereRadius (Variables|Default|GetInner%20Ring%20Radius%20Control)) (Math|Trig|Sin(Radians) (LocalVariables|GetAngle)))) false false)
        (Transformation|SetRelativeRotation _array_element (Math|Rotator|MakeRotator (* _array_index -3e+01)) false false)
        (LocalVariables|SetAngle (+ (LocalVariables|GetAngle) (/ (* 2.0 (Math|Trig|GetPI)) 12.0)))
        (Transformation|AddLocalOffset _array_element (Variables|Default|GetParticle_Position_Offset) false false))
      (LocalVariables|SetAngle (/ (Math|Trig|GetPI) 2.0)))
    (:then_1
      ;[1 1 1] Set Trans. on Ice Sphere
      (Rendering|Material|SetScalarParameterValueonMaterials (Variables|Default|GetIce%20Bubble) "Fade In Sphere" (Variables|Default|GetIce%20Transparency)))
    (:then_2
      ;[1 1 1] DEBUG
      (if (Variables|Default|GetDEBUG)
        (for _array_element_2 (Variables|Default|GetInner_TimeOrbs)
          (Components|Activation|SetAutoActivate _array_element_2 true))
        (Rendering|SetVisibility (Variables|Default|GetIce%20Bubble) true false)
        (Components|Activation|SetAutoActivate (Variables|Default|GetIdle) true)))
    (:then_3
      (if (not (ScalableFloat|GetValueasBool (Variables|Default|GetPropCheck)))
        (Components|DestroyComponent (Variables|Default|GetBuildingPropChecker))))))
