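; Mooney fog wall Blueprint that drives the cosmetic audio, camera shake and snow effects as the local view target nears and crosses it.
; A looping timer moves wall audio and a runtime-added shake component to the closest point on the wall line.
; A dot-product test triggers a one-time pass-by sound and timeline, stops tracking, then spawns attached snow and, outside replays, a frost lens effect.
; Replay scrubbing reattaches the snow and silences the shake.
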(event Custom|ReplayScrubbed
  ;[1 1 1] Try to fix things when scrubbing replay
  (Utilities|FlowControl|Sequence
    (:then_0
      (bind _spawnedsnowfx (Variables|Default|GetSpawnedSnowFX)
        (:then
          (Transformation|AttachComponentToComponent _spawnedsnowfx (Variables|/Script/Engine.Actor|GetRootComponent (Pawn|GetViewTarget (Utilities|Array|Get(acopy) (Fortnite|GetLocalFortPlayerControllers true true)))) "None" "SnapToTarget" "KeepWorld" "KeepWorld" true)
          (Variables|/Script/Engine.SceneComponent|SetbAbsoluteRotation true _spawnedsnowfx)
          (Transformation|SetWorldRotation _spawnedsnowfx "0, 0, 0" false false))
        (:else)))
    (:then_1
      (bind _camerashake (Variables|Default|GetCameraShake)
        (:then
          (Variables|/Figment_S07/Athena/Maps/Test/Events/CameraShakeComponent.CameraShakeComponent_C|SetShakeStrength 0.0 _camerashake))
        (:else)))))

(event EventBeginPlay
  (bind _returnvalue (BlueprintContext|GetFortReplayContext))
  (Utilities|FlowControl|Sequence
    (:then_0
      ;[1 1 1] Try to fix things when scrubbing replay
      (Utilities|IsValid _returnvalue
        (:Is%20Valid
          (ReplayContext|BindEventtoOnReplayScrubComplete _returnvalue (EventDispatchers|CreateEvent :!event Custom|ReplayScrubbed)))))
    (:then_1
      ;[1 1 1] Kick off timer for closest point on wall
      (Networking|CanExecuteCosmeticEvents
        (:True
          (bind _returnvalue_2 (Utilities|Time|SetTimerbyEvent (EventDispatchers|CreateEvent :!event Custom|On%20Audio%20Timer%20Loop) (Variables|Default|GetUpdateFrequency) true))
          (Variables|Default|SetAudioTimerHandle _returnvalue_2)
          (bind _returnvalue_3 (AddComponent|NODE_AddCameraShakeComponent-3))
          (Variables|Default|SetCameraShake _returnvalue_3))))))

(event Custom|On%20Audio%20Timer%20Loop
  (bind _returnvalue (Fortnite|GetLocalFortPlayerControllers true true))
  (bind _output (Utilities|Array|Get(acopy) _returnvalue))
  (bind _returnvalue_2 (Pawn|GetViewTarget _output))
  (bind _returnvalue_3 (Fortnite|GetLocalFortPlayerControllers true true))
  (bind _output_2 (Utilities|Array|Get(acopy) _returnvalue_3))
  (bind _returnvalue_4 (Pawn|GetViewTarget _output_2))
  (bind _wallbase (Variables|Default|GetWallBase))
  (bind _returnvalue_5 (Transformation|GetRightVector _wallbase))
  (bind _returnvalue_6 (Transformation|GetWorldLocation _wallbase))
  (bind _returnvalue_7 (Transformation|GetActorLocation _returnvalue_2))
  (bind _returnvalue_8 (Math|Vector|FindClosestPointonLine _returnvalue_7 _returnvalue_6 _returnvalue_5))
  (bind _viewtargetlocation (Variables|Default|GetViewTargetLocation))
  (bind _z (.z _viewtargetlocation))
  (bind _y (.y _returnvalue_8))
  (bind _x (.x _returnvalue_8))
  (bind _returnvalue_9 (Math|Vector|MakeVector _x _y _z))
  (bind _camerashake (Variables|Default|GetCameraShake))
  (bind _camerashake_2 (Variables|Default|GetCameraShake))
  ;[1 1 1] Logic to find closest point on line and position audio and FX there
  (Utilities|IsValid _output
    (:Is%20Valid
      (Utilities|IsValid _returnvalue_2
        (:Is%20Valid
          (Variables|Default|SetViewTargetLocation _returnvalue_7)
          (Variables|Default|SetAudioLocation _returnvalue_9)
          (bind _output_get (Variables|Default|SetPlayerInFront (>= (Math|Vector|DotProduct (Math|Vector|Normalize (- (+ (Variables|Default|GetViewTargetLocation) (* (Transformation|GetForwardVector (Variables|Default|GetWallBase)) (* (Variables|Default|GetDistanceFromWallBeforeSound) -1.0))) _returnvalue_9)) (Transformation|GetForwardVector (Variables|Default|GetWallBase))) 0.0)))
          (Transformation|SetWorldLocation (Variables|Default|GetWallAudio) (Variables|Default|GetAudioLocation) false false)
          ;[1 1 1] Camera shake fade
          (Variables|/Figment_S07/Athena/Maps/Test/Events/CameraShakeComponent.CameraShakeComponent_C|SetLocation (Variables|Default|GetAudioLocation) _camerashake)
          (Variables|/Figment_S07/Athena/Maps/Test/Events/CameraShakeComponent.CameraShakeComponent_C|SetShakeStrength 0.5 _camerashake)
          (if (not _output_get)
            (Utilities|FlowControl|DoOnce false
              (:Completed
                (Utilities|FlowControl|Sequence
                  (:then_0
                    (Audio|PlaySound2D (Variables|Default|GetPassBySound))
                    (Components|Timeline|SetPlayRate (Variables|BP_Mooney_Fog_Wall|GetPass%20By%20Timeline) (/ 1.0 1e+01))
                    (bind (_direction _passby) (Timeline|Pass%20By%20Timeline PlayFromStart)
                      (:Update
                        (Audio|Parameter|SetFloatParameter (Variables|Default|GetWallAudio) "PassBy" _passby :!class /Script/Engine.AudioComponent)
                        (Variables|/Figment_S07/Athena/Maps/Test/Events/CameraShakeComponent.CameraShakeComponent_C|SetShakeStrength (Math|Float|Lerp 0.0 0.5 _passby) _camerashake_2))
                      (:Finished
                        (Utilities|Time|ClearandInvalidateTimerbyHandle (Variables|Default|GetAudioTimerHandle))
                        (Components|DestroyComponent _camerashake_2))))
                  (:then_1
                    ;[1 1 1] Spawn FX and camera lens effect
                    (Utilities|FlowControl|Delay 0.5)
                    (Utilities|FlowControl|Sequence
                      (:then_0
                        (bind _returnvalue_10 (Effects|SpawnEmitterAttached :EmitterTemplate "/Figment_S07/Athena/Prototype/Blueprints/Mooney/FX/P_Player_Snow_Swirl.P_Player_Snow_Swirl" :AttachToComponent (Variables|/Script/Engine.Actor|GetRootComponent _returnvalue_4) :AttachPointName "None" :Scale "1.000000,1.000000,1.000000" :LocationType "SnapToTarget" :bAutoDestroy true :PoolingMethod "None" :bAutoActivate true))
                        (bind _output_get_2 (Variables|Default|SetSpawnedSnowFX _returnvalue_10))
                        (Variables|/Script/Engine.SceneComponent|SetbAbsoluteRotation true _output_get_2)
                        (Transformation|SetWorldRotation _output_get_2 "0, 0, 0" false false)
                        (Rendering|SetTranslucentSortPriority _output_get_2 2))
                      (:then_1
                        ;;[1 1 1] Camera Lens Effect
                        (Utilities|IsValid _returnvalue_4
                          (:Is%20Valid
                            (if (not (Replay|GetIsPlayingReplay))
                              (Feedback|SpawnGenericCameraLensEffect (Game|Player|GetPlayerPawn _output_2) "/Figment_S07/Athena/Prototype/Blueprints/Mooney/FX/BP_NS_Figment_CameraLens_IceFrost.BP_NS_Figment_CameraLens_IceFrost_C"))))))))))))))))

(fn ConstructionScript)
