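; BP_CameraLens_IceSphere is a camera lens particle effect that drives a screen material for an Ice King ice sphere.
; Its custom event Pass Particle Parameters creates a dynamic instance of the emitter's named "ScreenMat" material on each call.
; It writes the sphere center and radius into that instance, and maps the Ice King visibility flag to the CountdownStarted scalar as one or zero.
; The construction script is empty.
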(event Custom|Pass%20Particle%20Parameters
  (params (Ice%20Center :type Vector) (Radius :type Float%20%28double-precision%29) (IceKing%20Visible :type Boolean))
  (bind _returnvalue (Rendering|Material|CreateNamedDynamicMaterialInstance (Variables|Emitter|GetParticleSystemComponent) "ScreenMat" 0))
  (Rendering|Material|SetVectorParameterValue _returnvalue "Ice Center" (Math|Conversions|ToLinearColor(Vector) Ice%20Center) :!class /Script/Engine.MaterialInstanceDynamic)
  (Rendering|Material|SetScalarParameterValue _returnvalue "Radius" Radius :!class /Script/Engine.MaterialInstanceDynamic)
  (Rendering|Material|SetScalarParameterValue _returnvalue "CountdownStarted" (select IceKing%20Visible 1.0 0.0) :!class /Script/Engine.MaterialInstanceDynamic))

(fn ConstructionScript)
