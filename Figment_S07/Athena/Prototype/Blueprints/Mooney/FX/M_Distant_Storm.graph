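# Translucent unlit mesh-particle material that draws distant storm clouds with vertically stretched, panning textures, timed by particle dynamic parameters.
# Cloud normals dotted with the sun/moon vector drive time-of-day highlight and shadow colors, with a cheaper night-tint blend on low quality and mobile.
# The cloud mask drives normal-based vertex displacement and opacity, both shaped by sphere-mask and view-angle fades.
# Opacity also uses depth and near-camera fades.
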Material /Figment_S07/Athena/Prototype/Blueprints/Mooney/FX/M_Distant_Storm
graph {
 properties {bUsedWithMeshParticles=true, blendMode=BLEND_Translucent, pixelDepthOffsetMode=PDOM_Legacy, shadingModel=MSM_Unlit, usedShadingModels=Unlit}
 nodes [
  n1 = CollectionParameter() {
   collection=&FortniteMaterialParameters
   parameterName=SunAndMoonModelDirectionalVector
  }
  n2 = Normalize(VectorInput=ComponentMask(None=@n1) {b=true, g=true, r=true})
  n3 = ComponentMask(None=VectorParameter() {parameterName=CloudTile, defaultValue=<rgba 0.6 2 0 0>}.RGB) {g=true, r=true}
  n4 = TextureCoordinate()
  n5 = Constant() {r=1}
  n6 = TextureCoordinate()
  n7 = LinearInterpolate(
   A=ScalarParameter() {defaultValue=0.9, parameterName=StretchMin}
   B=ScalarParameter() {defaultValue=1.8, parameterName=StretchMax}
   Alpha=Power(Base=ComponentMask(None=@n6) {g=true}, Exp=ScalarParameter() {defaultValue=8, parameterName=StretchPower})
  ) {constA=0.25, constB=1.5}
  n8 = Time()
  n9 = DynamicParameter() {paramNames [PanAdd,PanSpeed,Param3,Param4]}
  n10 = Panner(
   Coordinate=Multiply(A=Multiply(A=@n3, B=@n4), B=AppendVector(A=@n5, B=@n7))
   Time=Multiply(A=Add(A=@n8, B=@n9.PanAdd), B=@n9.PanSpeed)
   Speed=ComponentMask(None=VectorParameter() {parameterName=PanDirection, defaultValue=<rgba -0.2 0.5 0 0>}.RGB) {g=true, r=true}
  ) {speedX=0.05, speedY=0.5}
  n11 = TextureSample(UVs=@n10) {samplerType=SAMPLERTYPE_Normal, texture=/Game/VisualThreatManager/StormVisuals/LargeClouds/T_StormCloudDetail01_N}
  n12 = ScalarParameter() {parameterName=NormalFlatness}
  n13 = Transform(None=Multiply(
    A=VectorParameter() {parameterName=NormalTransform, defaultValue=<rgba -1 1 1 0>}.RGB
    B=MaterialFunctionCall(Normal=@n11.RGB, Flatness=@n12) {materialFunction=&FlattenNormal}.Result
   ))
  n14 = Power(Base=Clamp(None=DotProduct(A=@n2, B=@n13)), Exp=ScalarParameter() {defaultValue=2, parameterName=HighlightPower})
  n15 = VectorParameter() {group=Colors, parameterName=NightTint, defaultValue=<rgba 0.3318 0.4932 0.7 1>}
  n16 = Multiply(
   A=VectorParameter() {group=Colors, parameterName=HighlightColor, defaultValue=<rgba 0.7257 1.2 0.6969 0>}.RGB
   B=MaterialFunctionCall(
    Color_Evening=VectorParameter() {group=Colors, parameterName=EveningTint, defaultValue=<rgba 0.651 0.4035 0.4426 1>}.RGB
    Color_Night=@n15.RGB
   ) {materialFunction=&Color_Time_of_Day}.Result
  )
  n17 = VectorParameter() {group=Colors, parameterName=ShadowNight, defaultValue=<rgba 0.4684 0.5912 0.7 0>}
  n18 = MaterialFunctionCall(
   Color_Morning=VectorParameter() {group=Colors, parameterName=ShadowMorning, defaultValue=<rgba 0.9 0.5487 0.4078 0>}.RGB
   Color_Day=VectorParameter() {group=Colors, parameterName=ShadowDay, defaultValue=<rgba 0.4531 0.5749 1 0>}.RGB
   Color_Evening=VectorParameter() {group=Colors, parameterName=ShadowEvening, defaultValue=<rgba 0.2116 0.2276 0.4 1>}.RGB
   Color_Night=@n17.RGB
  ) {materialFunction=&Color_Time_of_Day}
  n19 = ScalarParameter() {defaultValue=0.6, parameterName=ShadowMult}
  n20 = TextureSample(UVs=@n10) {samplerType=SAMPLERTYPE_LinearGrayscale, texture=/Game/VisualThreatManager/StormVisuals/LargeClouds/T_StormCloudDetail01_M}
  n21 = Multiply(A=@n20.G)
  n22 = LinearInterpolate(A=@n17.RGB, B=@n15.RGB, Alpha=@n21)
  n23 = QualitySwitch(Default=Add(A=Multiply(A=@n14, B=@n16) {constB=0}, B=Multiply(A=@n18.Result, B=@n19)), Low=@n22)
  n24 = ParticleColor()
  Multiply(
   A=Multiply(A=FeatureLevelSwitch(Default=@n23, ES2=@n22, ES3_1=@n22, SM4_REMOVED=@n23), B=LinearInterpolate(Alpha=@n21) {constA=0.35, constB=0.75})
   B=@n24.RGB
  )
  n25 = Constant() {r=0.5}
  n26 = SphereMask(A=TextureCoordinate(), B=@n25, Radius=@n25, Hardness=ScalarParameter() {parameterName=SideFadeHardness}) {attenuationRadius=0.5, hardnessPercent=50}
  n27 = Multiply(A=VertexNormalWS(), B=LinearInterpolate(Alpha=@n20.G) {constA=-0.8})
  n28 = ScalarParameter() {defaultValue=10, parameterName=WPOAmount}
  n29 = StaticBool()
  n30 = Abs(None=DotProduct(
    A=MaterialFunctionCall!(Use World Position Offsets=@n29) {materialFunction=/Engine/Functions/Engine_MaterialFunctions02/Utility/CameraVectorWithWPOOptions}.Result
    B=VertexNormalWS()
   ))
  Multiply(A=Multiply(A=@n26, B=Multiply(A=@n27, B=@n28)), B=Clamp(None=Power(Base=@n30) {constExponent=0.6}))
  n31 = ScalarParameter() {parameterName=OpacityMin}
  n32 = ScalarParameter() {defaultValue=1, parameterName=OpacityMax}
  n33 = Multiply(A=@n21, B=ComponentMask(None=TextureCoordinate()) {g=true})
  n34 = ScalarParameter() {defaultValue=20000, parameterName=DepthFadeMin}
  n35 = ScalarParameter() {defaultValue=80000, parameterName=DepthFadeMax}
  n36 = Multiply(A=OneMinus(None=@n20.G), B=@n26)
  n37 = Multiply(
   A=DepthFade(Opacity=@n26, FadeDistance=LinearInterpolate(A=@n34, B=@n35, Alpha=@n36))
   B=Power(Base=@n30, Exp=ScalarParameter() {defaultValue=1.7, parameterName=AngleFadePower})
  )
  n38 = MaterialFunctionCall(Distance=ScalarParameter() {defaultValue=25000, parameterName=NearFadeDistance}) {materialFunction=&Near_Camera_Fade}
  Multiply(A=@n24.A, B=Multiply(A=LinearInterpolate(A=@n31, B=@n32, Alpha=@n33) {constA=1.4}, B=Multiply(A=@n37, B=@n38.Result)))
 ]
}
