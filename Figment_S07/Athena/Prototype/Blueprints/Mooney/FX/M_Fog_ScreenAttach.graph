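# Unlit translucent fog overlay material that attaches a mesh to the camera, fading in while the camera is between front and rear planes along the actor's local X axis.
# Outside that band, a large world position offset pushes the mesh away.
# View-direction projected storm cloud samples blend by axis, with panning and two-phase zoom, driving opacity and a height- and sun-tinted color.
# Mobile feature levels skip depth fade and add a normal push.
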Material /Figment_S07/Athena/Prototype/Blueprints/Mooney/FX/M_Fog_ScreenAttach
graph {
 properties {
  bUseLightmapDirectionality=false
  blendMode=BLEND_Translucent
  pixelDepthOffsetMode=PDOM_Legacy
  shadingModel=MSM_Unlit
  usedShadingModels=Unlit
 }
 nodes [
  n1 = ActorPositionWS()
  n2 = Transform(None=Constant3Vector() {constant=<rgba 1 0 0 0>}) {transformSourceType=TRANSFORMSOURCE_Local}
  n3 = Multiply(A=ScalarParameter() {defaultValue=14000, parameterName=WallFrontOffset}, B=@n2)
  n4 = CameraPositionWS()
  n5 = DotProduct(A=Subtract(A=Add(A=@n1, B=@n3), B=@n4), B=@n2)
  n6 = ScalarParameter() {defaultValue=10000, parameterName=WallFrontFadeLength}
  n7 = Add(A=@n1, B=Multiply(A=@n2, B=ScalarParameter() {defaultValue=-100000, parameterName=WallRearOffset}))
  n8 = Divide(A=DotProduct(A=Subtract(A=@n7, B=@n4), B=@n2), B=ScalarParameter() {defaultValue=50000, parameterName=WallRearFadeLength})
  n9 = Multiply(A=Clamp(None=Divide(A=@n5, B=@n6)), B=OneMinus(None=Clamp(None=@n8)))
  n10 = Ceil(None=@n9)
  n11 = Add(
   A=MaterialFunctionCall!(
    Object Basis 1
    Object Basis 2
    Object Basis 3
    Camera Space Position Offset=VectorParameter() {parameterName=ScreenPosition, defaultValue=<rgba 0 0 12 1>}.RGB
   ) {materialFunction=/Engine/Functions/Engine_MaterialFunctions02/WorldPositionOffset/AttachMeshToTheCamera}.World Position Offset
   B=Multiply(A=OneMinus(None=@n10), B=Constant3Vector() {constant=<rgba 0 0 -54654566400 0>})
  )
  n12 = Add(A=@n11, B=Multiply(A=VertexNormalWS(), B=ScalarParameter() {defaultValue=-3714.3762, parameterName=MobileOffsetPush}))
  FeatureLevelSwitch(Default=@n11, ES2=@n12, ES3_1=@n12, SM4_REMOVED=@n11)
  n13 = VectorParameter() {parameterName=Midtone, defaultValue=<rgba 0.0762 0.3121 0.5417 1>}
  n14 = Constant3Vector() {constant=<rgba 0 0 1 0>}
  n15 = ScalarParameter() {defaultValue=0.37, parameterName=RotationAngle}
  n16 = Constant3Vector()
  n17 = WorldPosition()
  n18 = CameraPositionWS()
  n19 = Normalize(VectorInput=Subtract(A=@n17.XYZ, B=@n18))
  n20 = Add(
   A=RotateAboutAxis(NormalizedRotationAxis=@n14, RotationAngle=@n15, PivotPoint=@n16, Position=@n19)
   B=ComponentMask(None=@n19) {b=true, g=true, r=true}
  ) {desc=Direction}
  n21 = ComponentMask(None=@n20) {b=true}
  n22 = Add(A=@n21)
  n23 = LinearInterpolate(A=VectorParameter() {parameterName=Shadow, defaultValue=<rgba 0 0.0947 0.1667 1>}.RGB, B=@n13.RGB, Alpha=@n22)
  n24 = VectorParameter() {parameterName=Highlight, defaultValue=<rgba 0.5417 0.9276 1.25 1>}
  n25 = CollectionParameter() {
   collection=&FortniteMaterialParameters
   parameterName=SunlightDirectionalVector
  }
  n26 = Clamp(None=DotProduct(A=@n19, B=ComponentMask(None=@n25) {b=true, g=true, r=true}))
  n27 = Subtract(A=ScalarParameter() {defaultValue=21379.3926, parameterName=ProjectionDistance}, B=Distance(A=@n17.XYZ, B=@n18))
  n28 = Subtract(A=Add(A=@n17.XYZ, B=Multiply(A=@n27, B=@n19)), B=@n18)
  n29 = Add(A=RotateAboutAxis(NormalizedRotationAxis=@n14, RotationAngle=@n15, PivotPoint=@n16, Position=@n28), B=@n28) {desc=UVs}
  n30 = ComponentMask(None=@n29) {r=true}
  n31 = ComponentMask(None=@n29) {b=true}
  n32 = VectorParameter() {parameterName=PanDirection, defaultValue=<rgba -1 0 0.1143 1>}
  n33 = Multiply(A=Time() {period=5}, B=ScalarParameter() {defaultValue=1, parameterName=Speed})
  n34 = Add(
   A=Divide(A=AppendVector(A=@n30, B=@n31), B=ScalarParameter() {defaultValue=-20293.4453, parameterName=Scale})
   B=ComponentMask(None=Multiply(A=@n32.RGB, B=@n33)) {b=true, r=true}
  )
  n35 = ComponentMask(None=@n29) {g=true}
  n36 = AppendVector(A=@n35, B=@n31)
  n37 = ScalarParameter() {defaultValue=-4000, parameterName=Scale_2}
  n38 = ScalarParameter() {defaultValue=-33036.5312, parameterName=Scale_1}
  n39 = Multiply(A=Time() {period=5}, B=ScalarParameter() {defaultValue=0.4, parameterName=Speed_1})
  n40 = Frac(None=@n39)
  n41 = ComponentMask(None=@n20) {r=true}
  n42 = Clamp(None=Sign(None=@n41))
  n43 = LinearInterpolate(A=OneMinus(None=@n40), B=@n40, Alpha=@n42)
  n44 = TextureSample(UVs=Divide(A=@n36, B=LinearInterpolate(A=@n37, B=@n38, Alpha=@n43))) {samplerType=SAMPLERTYPE_Masks, texture=/Game/Athena/SafeZone/Textures/StormCloudtubes_M}
  n45 = Frac(None=Add(A=@n39) {constB=0.5})
  n46 = LinearInterpolate(A=@n37, B=@n38, Alpha=LinearInterpolate(A=OneMinus(None=@n45), B=@n45, Alpha=@n42))
  n47 = TextureSample(UVs=Add(A=Divide(A=@n36, B=@n46), B=Constant2Vector() {g=0.743, r=0.342})) {samplerType=SAMPLERTYPE_Masks, texture=&StormCloudtubes_M}
  n48 = Cosine(None=@n39)
  n49 = Clamp(None=Divide(A=Add(A=@n48)))
  n50 = Divide(A=AppendVector(A=@n35, B=@n30), B=ScalarParameter() {defaultValue=-29100.7031, parameterName=Scale_5})
  n51 = ComponentMask(None=Multiply(A=@n33, B=VectorParameter() {parameterName=PanDirection2_1, defaultValue=<rgba 0 -0.6762 0 1>}.RGB)) {g=true, r=true}
  n52 = LinearInterpolate(
   A=LinearInterpolate(
    A=TextureSample(UVs=@n34) {samplerType=SAMPLERTYPE_Masks, texture=&StormCloudtubes_M}.B
    B=LinearInterpolate(A=@n44.B, B=@n47.B, Alpha=@n49)
    Alpha=Abs(None=@n41)
   )
   B=TextureSample(UVs=Add(A=@n50, B=@n51)) {samplerType=SAMPLERTYPE_Masks, texture=&StormCloudtubes_M}.B
   Alpha=Abs(None=@n21)
  )
  LinearInterpolate(
   A=VectorParameter() {parameterName=Color_1, defaultValue=<rgba 0.5455 0.682 0.8802 1>}.RGB
   B=QualitySwitch(Default=LinearInterpolate(A=@n23, B=@n24.RGB, Alpha=@n26), Low=@n13.RGB)
   Alpha=OneMinus(None=@n52)
  )
  n53 = LinearInterpolate(
   A=LinearInterpolate(
    A=ScalarParameter() {defaultValue=-0.15, parameterName=Opacity_1}
    B=ScalarParameter() {defaultValue=2.5, parameterName=Opacity_2}
    Alpha=@n52
   )
   B=LinearInterpolate(A=ScalarParameter() {defaultValue=-0.2, parameterName=Opacity_3}, B=ScalarParameter() {defaultValue=3, parameterName=Opacity_4})
   Alpha=Divide(A=@n22)
  )
  n54 = DepthFade(Opacity=@n53, FadeDistance=ScalarParameter() {defaultValue=5000, parameterName=DepthFade})
  Multiply(A=VertexInterpolator(VS=@n9).PS, B=FeatureLevelSwitch(Default=@n54, ES2=@n53, ES3_1=@n53, SM4_REMOVED=@n54))
 ]
}
