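# Unlit translucent material, usable on particle sprites, that draws an animated fog wall from panning storm cloud textures.
# Two noise layers shape brightness and opacity, while the first also pushes vertices along their normals.
# Flattened normals lit by the sun/moon vector blend time-of-day highlight and shadow colors.
# Edge, depth, view-angle and near-camera fades soften it; low quality and mobile paths simplify color and depth fade, with mobile dropping displacement.
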Material /Figment_S07/Athena/Prototype/Blueprints/Mooney/FX/M_Fog_Wall
graph {
 properties {
  bUsedWithParticleSprites=true
  blendMode=BLEND_Translucent
  pixelDepthOffsetMode=PDOM_Legacy
  shadingModel=MSM_Unlit
  usedShadingModels=Unlit
 }
 nodes [
  n1 = TextureCoordinate()
  n2 = ComponentMask(None=@n1) {g=true}
  n3 = ComponentMask(None=@n1) {r=true}
  n4 = ScalarParameter() {defaultValue=1, parameterName=SideFadePower}
  n5 = Constant() {r=0.5}
  n6 = Multiply(
   A=@n2
   B=SphereMask(A=Power(Base=@n3, Exp=@n4), B=@n5, Radius=@n5, Hardness=ScalarParameter() {defaultValue=0.7, parameterName=SideFadeHardness}) {attenuationRadius=0.5, hardnessPercent=50}
  )
  n7 = ComponentMask(None=VectorParameter() {parameterName=CloudTile, defaultValue=<rgba 3 0.8 0 0>}.RGB) {g=true, r=true}
  n8 = TextureCoordinate()
  n9 = Constant() {r=1}
  n10 = TextureCoordinate()
  n11 = LinearInterpolate(
   A=ScalarParameter() {defaultValue=1.2, parameterName=StretchMin}
   B=ScalarParameter() {defaultValue=2, parameterName=StretchMax}
   Alpha=Power(Base=ComponentMask(None=@n10) {g=true}, Exp=ScalarParameter() {defaultValue=8, parameterName=StretchPower})
  ) {constA=0.25, constB=1.5}
  n12 = Multiply(A=Add(A=Time(), B=ScalarParameter() {parameterName=Random}), B=ScalarParameter() {defaultValue=0.5, parameterName=PanSPeed})
  n13 = Panner(
   Coordinate=Multiply(A=Multiply(A=@n7, B=@n8), B=AppendVector(A=@n9, B=@n11))
   Time=@n12
   Speed=ComponentMask(None=VectorParameter() {parameterName=PanDirection, defaultValue=<rgba 0.05 -0.5 0 0>}.RGB) {g=true, r=true}
  ) {speedX=0.05, speedY=-0.5}
  n14 = TextureSample(UVs=@n13) {samplerType=SAMPLERTYPE_LinearGrayscale, texture=/Game/VisualThreatManager/StormVisuals/LargeClouds/T_StormCloudDetail01_M}
  n15 = Multiply(A=VertexNormalWS(), B=LinearInterpolate(Alpha=@n14.G) {constA=-0.8})
  n16 = ScalarParameter() {defaultValue=6400, parameterName=WPOAmount}
  n17 = DotProduct(
   A=MaterialFunctionCall!(Use World Position Offsets=StaticBool()) {materialFunction=/Engine/Functions/Engine_MaterialFunctions02/Utility/CameraVectorWithWPOOptions}.Result
   B=VertexNormalWS()
  )
  n18 = Multiply(A=Multiply(A=@n6, B=Multiply(A=@n15, B=@n16)), B=Clamp(None=Power(Base=@n17) {constExponent=0.6}))
  n19 = Constant()
  FeatureLevelSwitch(Default=@n18, ES2=@n19, ES3_1=@n19, SM4_REMOVED=@n18)
  n20 = CollectionParameter() {
   collection=&FortniteMaterialParameters
   parameterName=SunAndMoonModelDirectionalVector
  }
  n21 = Normalize(VectorInput=ComponentMask(None=@n20) {b=true, g=true, r=true})
  n22 = TextureSample(UVs=@n13) {samplerType=SAMPLERTYPE_Normal, texture=/Game/VisualThreatManager/StormVisuals/LargeClouds/T_StormCloudDetail01_N}
  n23 = CameraPositionWS()
  n24 = WorldPosition()
  n25 = Clamp(None=Divide(A=Distance(A=@n23, B=@n24.XYZ), B=ScalarParameter() {defaultValue=250000, parameterName=DistanceBlend}))
  n26 = Clamp(None=@n25) {maxDefault=0.7}
  n27 = Transform(None=Multiply(
    A=VectorParameter() {parameterName=NormalTransform, defaultValue=<rgba -1 1 1 0>}.RGB
    B=MaterialFunctionCall(Normal=@n22.RGB, Flatness=@n26) {materialFunction=&FlattenNormal}.Result
   ))
  n28 = Power(Base=Clamp(None=DotProduct(A=@n21, B=@n27)), Exp=ScalarParameter() {defaultValue=2, parameterName=HighlightPower})
  n29 = VectorParameter() {group=Colors, parameterName=NightTint, defaultValue=<rgba 0.5214 0.6719 0.8073 1>}
  n30 = Multiply(
   A=VectorParameter() {group=Colors, parameterName=HighlightColor, defaultValue=<rgba 1.0271 1.2661 1.7 1>}.RGB
   B=MaterialFunctionCall(Color_Evening=VectorParameter() {group=Colors, parameterName=EveningTint, defaultValue=<rgba 0.7 0.4265 0.3916 1>}.RGB, Color_Night=@n29.RGB) {materialFunction=&Color_Time_of_Day}.Result
  )
  n31 = VectorParameter() {group=Colors, parameterName=ShadowNight, defaultValue=<rgba 0.0902 0.3187 0.5972 1>}
  n32 = MaterialFunctionCall(
   Color_Morning=VectorParameter() {group=Colors, parameterName=ShadowMorning, defaultValue=<rgba 0.9 0.5487 0.4078 0>}.RGB
   Color_Day=VectorParameter() {group=Colors, parameterName=ShadowDay, defaultValue=<rgba 0.4531 0.5749 1 0>}.RGB
   Color_Evening=VectorParameter() {group=Colors, parameterName=ShadowEvening, defaultValue=<rgba 0.3107 0.2373 0.4 1>}.RGB
   Color_Night=@n31.RGB
  ) {materialFunction=&Color_Time_of_Day}
  n33 = ScalarParameter() {defaultValue=1, parameterName=ShadowMult}
  n34 = Multiply(
   A=ComponentMask(None=VectorParameter() {parameterName=CloudTile2, defaultValue=<rgba 2.6 2 0 0>}.RGB) {g=true, r=true}
   B=TextureCoordinate()
  )
  n35 = Add(A=TextureSample(UVs=Panner(Coordinate=@n34, Time=@n12) {speedX=-0.05, speedY=-0.6}) {samplerType=SAMPLERTYPE_LinearGrayscale, texture=&T_StormCloudDetail01_M}.G) {constB=0.25}
  n36 = Multiply(A=@n14.G, B=QualitySwitch(Default=LinearInterpolate(A=@n35, Alpha=@n25) {constA=0.25}, Low=Constant() {r=1}))
  n37 = LinearInterpolate(A=@n31.RGB, B=@n29.RGB, Alpha=@n36)
  n38 = QualitySwitch(Default=Add(A=Multiply(A=@n28, B=@n30) {constB=0}, B=Multiply(A=@n32.Result, B=@n33)), Low=@n37)
  Multiply(A=FeatureLevelSwitch(Default=@n38, ES2=@n37, ES3_1=@n37, SM4_REMOVED=@n38), B=LinearInterpolate(Alpha=@n36) {constA=0.4, constB=0.8})
  n39 = ComponentMask(None=TextureCoordinate()) {g=true}
  n40 = ScalarParameter() {defaultValue=4000, parameterName=DepthFadeMin}
  n41 = ScalarParameter() {defaultValue=30000, parameterName=DepthFadeMax}
  n42 = Multiply(A=OneMinus(None=@n14.G), B=@n2)
  n43 = QualitySwitch(Default=Multiply(A=LinearInterpolate(Alpha=@n25) {constA=1, constB=5}, B=LinearInterpolate(A=@n40, B=@n41, Alpha=@n42)), Low=@n41)
  n44 = Multiply(
   A=DepthFade(Opacity=@n6, FadeDistance=FeatureLevelSwitch(Default=@n43, ES2=@n41, ES3_1=@n41, SM4_REMOVED=@n43))
   B=Power(Base=@n17, Exp=ScalarParameter() {defaultValue=1, parameterName=AngleFadePower})
  )
  n45 = MaterialFunctionCall(Distance=ScalarParameter() {defaultValue=25000, parameterName=NearFadeDistance}) {materialFunction=&Near_Camera_Fade}
  Multiply(
   A=LinearInterpolate(
    A=ScalarParameter() {defaultValue=1, parameterName=OpacityMin}
    B=ScalarParameter() {defaultValue=1.7, parameterName=OpacityMax}
    Alpha=Multiply(A=@n36, B=@n39)
   ) {constA=1.4}
   B=Multiply(A=ScalarParameter() {defaultValue=1.5, parameterName=OpacityMult}, B=Multiply(A=@n44, B=@n45.Result))
  )
 ]
}
