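# Translucent unlit screen overlay that attaches a mesh to the camera and shows frost and whiteout as the camera nears the actor.
# World position offset throws the mesh far away when out of range, with an extra normal push on mobile.
# Tiled screen-aligned frost closes in from the edges, using fewer samples on mobile and low quality.
# MF_SharedWinter supplies colors, and CountdownStarted fades out whiteout while altering the center mask.
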Material /Figment_S07/Athena/Prototype/Blueprints/Mooney/FX/M_Frost_ScreenAttach
graph {
 properties {
  bDisableDepthTest=true
  bUseLightmapDirectionality=false
  blendMode=BLEND_Translucent
  pixelDepthOffsetMode=PDOM_Legacy
  shadingModel=MSM_Unlit
  usedShadingModels=Unlit
 }
 nodes [
  n1 = Subtract(A=Distance(A=CameraPositionWS(), B=ActorPositionWS()), B=ScalarParameter() {defaultValue=400, parameterName=FadeStartOffset})
  n2 = ScalarParameter() {defaultValue=2500, parameterName=Radius}
  n3 = OneMinus(None=Clamp(None=Divide(A=@n1, B=@n2)))
  n4 = Multiply(A=@n3)
  n5 = Ceil(None=@n4)
  n6 = Add(
   A=MaterialFunctionCall!(
    Object Basis 1
    Object Basis 2
    Object Basis 3
    Camera Space Position Offset=VectorParameter() {parameterName=ScreenPosition, defaultValue=<rgba 0 0 12 1>}.RGB
   ) {materialFunction=/Engine/Functions/Engine_MaterialFunctions02/WorldPositionOffset/AttachMeshToTheCamera}.World Position Offset
   B=Multiply(A=OneMinus(None=@n5), B=Constant3Vector() {constant=<rgba 0 0 -54654566400 0>})
  )
  n7 = Add(A=@n6, B=Multiply(A=VertexNormalWS(), B=ScalarParameter() {defaultValue=-3714.3762, parameterName=MobileOffsetPush}))
  FeatureLevelSwitch(Default=@n6, ES2=@n7, ES3_1=@n7, SM4_REMOVED=@n6)
  n8 = ScalarParameter() {defaultValue=6, parameterName=Frost_U_Tile}
  n9 = ScalarParameter() {defaultValue=6, parameterName=Frost_V_Tile}
  n10 = TextureSample(UVs=Multiply(
    A=MaterialFunctionCall() {materialFunction=&ScreenAlignedUVs}."X 100%, Y Scale to Ratio"
    B=AppendVector(A=@n8, B=@n9)
   )) {texture=/Game/Effects/Fort_Effects/Textures/Snowflake/T_LightFrost}
  n11 = Multiply(
   A=MaterialFunctionCall() {materialFunction=&ScreenAlignedUVs}."X 100%, Y Scale to Ratio"
   B=AppendVector(A=ScalarParameter() {defaultValue=10, parameterName=Frost_U_Tile2}, B=ScalarParameter() {defaultValue=10, parameterName=Frost_V_Tile2})
  )
  n12 = AppendVector(A=ScalarParameter() {defaultValue=0.06, parameterName=Frost_U_Offset}, B=ScalarParameter() {defaultValue=0.05, parameterName=Frost_V_Offset})
  n13 = Add(A=@n10.R, B=TextureSample(UVs=Add(A=@n11, B=@n12)) {texture=&T_LightFrost}.G)
  n14 = ScalarParameter() {defaultValue=25, parameterName=Frost_U_Tile3}
  n15 = ScalarParameter() {defaultValue=25, parameterName=Frost_V_Tile3}
  n16 = Add(
   A=Multiply(
    A=MaterialFunctionCall() {materialFunction=&ScreenAlignedUVs}."X 100%, Y Scale to Ratio"
    B=AppendVector(A=@n14, B=@n15)
   )
   B=AppendVector(A=ScalarParameter() {defaultValue=0.3, parameterName=Frost_U_Offset2, sliderMin=0.06}, B=ScalarParameter() {parameterName=Frost_V_Offset2})
  )
  n17 = Add(A=@n10.R, B=@n10.R)
  n18 = QualitySwitch(Default=Add(A=@n13, B=TextureSample(UVs=@n16) {texture=&T_LightFrost}.B), Low=@n17)
  n19 = ScalarParameter() {parameterName=CountdownStarted}
  n20 = OneMinus(None=SphereMask(
    A=MaterialFunctionCall() {materialFunction=&ScreenAlignedUVs}."X 100%, Y 100%"
    B=Constant() {r=0.5}
    Radius=LinearInterpolate(B=@n19, Alpha=@n3) {constA=6, constB=0}
   ) {hardnessPercent=25})
  n21 = Multiply(A=FeatureLevelSwitch(Default=@n18, ES2=@n17, ES3_1=@n17, SM4_REMOVED=@n18), B=@n20)
  n22 = ScalarParameter() {defaultValue=3, parameterName=Ratio_Opacity}
  n23 = ScalarParameter() {defaultValue=2500, parameterName=Radius}
  n24 = Clamp(None=SphereMask(A=WorldPosition().XYZ, B=ActorPositionWS(), Radius=Multiply(A=@n23, B=@n22)) {hardnessPercent=0})
  n25 = LinearInterpolate(A=Power(Base=@n3, Exp=ScalarParameter() {defaultValue=5, parameterName=WhiteoutPower}), Alpha=@n19) {constB=0}
  Multiply(A=VertexInterpolator(VS=@n4).PS, B=Add(A=Multiply(A=@n21, B=@n24), B=@n25))
  n26 = MaterialFunctionCall() {materialFunction=/Game/Athena/WinterArt/Materials/MF_SharedWinter}
  n27 = QualitySwitch(Default=@n13, Low=@n17)
  n28 = Multiply(A=FeatureLevelSwitch(Default=@n27, ES2=@n17, ES3_1=@n17, SM4_REMOVED=@n27), B=@n20)
  Multiply(A=Add(A=LinearInterpolate(A=@n26.IceColor, B=@n26.IceColor, Alpha=@n28), B=Multiply(A=@n26.SnowColor, B=@n25))) {constB=0.6}
 ]
}
