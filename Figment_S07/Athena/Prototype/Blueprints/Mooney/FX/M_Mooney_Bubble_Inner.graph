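# Lit material for the inner layer of the Mooney bubble effect, rendering a frosted, scratched ice surface.
# Base color blends ice, built from parallax-panned scratch normals and chromatic-offset frost, into snow through a height-lerp frost mask that also drives two other lerps, including metallic.
# Emissive adds scrolling, channel-split shield caustics and time-of-day-tinted highlights along and opposite the sun/moon direction.
# Shared winter functions supply the ice and snow colors.
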Material /Figment_S07/Athena/Prototype/Blueprints/Mooney/FX/M_Mooney_Bubble_Inner
graph {
 properties {
  ditherOpacityMask=true
  pixelDepthOffsetMode=PDOM_Legacy
  translucencyLightingMode=TLM_VolumetricPerVertexDirectional
  usedShadingModels=Default Lit
 }
 nodes [
  n1 = MaterialFunctionCall() {materialFunction=/Game/Athena/WinterArt/Materials/MF_SharedWinter}
  n2 = Add(
   A=AppendVector(A=ScalarParameter() {defaultValue=0.5, parameterName=Frost_U_Offset}, B=ScalarParameter() {defaultValue=0.5, parameterName=Frost_V_Offset})
   B=TextureCoordinate() {uTiling=2, vTiling=2}
  )
  n3 = ScalarParameter() {defaultValue=0.005, parameterName=Chroma Offset}
  n4 = AppendVector(A=@n3, B=@n3)
  n5 = TextureSample(UVs=Add(A=@n2, B=@n4)) {texture=/Game/Effects/Fort_Effects/Textures/Snowflake/T_LightFrost}
  n6 = TextureSample(UVs=@n2) {texture=&T_LightFrost}
  n7 = TextureSample(UVs=Add(A=@n2, B=Multiply(A=@n4) {constB=-1})) {texture=&T_LightFrost}
  n8 = AppendVector(A=AppendVector(A=@n5.G, B=@n6.G), B=@n7.G)
  n9 = AppendVector(A=AppendVector(A=@n5.B, B=@n6.B), B=@n7.B)
  n10 = Add(A=AppendVector(A=AppendVector(A=@n5.R, B=@n6.R), B=@n7.R), B=Add(A=Multiply(A=@n8), B=Multiply(A=@n9)))
  n11 = ComponentMask(None=TextureCoordinate()) {g=true}
  n12 = ScalarParameter() {defaultValue=0.5, parameterName=Depth of Frost Layer}
  n13 = Multiply(A=TextureCoordinate(), B=Constant() {r=1})
  n14 = TextureSample() {samplerType=SAMPLERTYPE_Masks, texture=/Game/Global/Textures/Default/Tileable_Patterns/T_IceScratch1_M}
  n15 = Constant() {r=-0.05}
  n16 = Time()
  n17 = ScalarParameter() {defaultValue=0.25, parameterName=SpeedScale}
  n18 = TextureSample(UVs=Panner(Coordinate=BumpOffset(Coordinate=@n13, Height=@n14.R, HeightRatioInput=@n15), Time=Multiply(A=@n16, B=@n17)) {speedX=-0.05, speedY=-0.05}) {samplerType=SAMPLERTYPE_Normal, texture=/Figment_S07/Athena/WinterArt/Textures/T_IceScratch1_N}
  n19 = DotProduct(A=CameraVectorWS(), B=Transform(None=@n18.RGB))
  n20 = Constant() {r=12}
  n21 = Multiply(A=Saturate(None=Power(Base=@n19, Exp=@n20)), B=Constant() {r=0.075})
  n22 = MaterialFunctionCall() {materialFunction=&MF_SharedWinter}
  n23 = LinearInterpolate(
   A=Multiply(A=Constant() {r=0.35}, B=@n22.IceColor)
   B=Multiply(A=@n22.IceColor, B=Constant3Vector() {constant=<rgba 2.2274 2.5 2.0045 0>})
   Alpha=TextureSample() {samplerType=SAMPLERTYPE_Masks, texture=&T_IceScratch1_M}.R
  )
  n24 = ScalarParameter() {defaultValue=0.5, parameterName=Frost_U_Offset}
  n25 = ScalarParameter() {defaultValue=0.5, parameterName=Frost_V_Offset}
  n26 = TextureSample(UVs=Add(A=AppendVector(A=@n24, B=@n25), B=TextureCoordinate() {uTiling=2, vTiling=2})) {texture=&T_LightFrost}
  n27 = ComponentMask(None=TextureCoordinate()) {g=true}
  n28 = Saturate(None=MaterialFunctionCall(
    Transition Phase=ScalarParameter() {defaultValue=0.4219, parameterName=Param_2}
    Height Texture=Multiply(A=@n26.R, B=@n27)
    Contrast=ScalarParameter() {defaultValue=0.25, parameterName=Conrast}
   ) {materialFunction=&HeightLerp}.Alpha)
  Add(
   A=Multiply(A=Multiply(A=@n1.IceColor, B=@n10), B=Power(Base=@n11, Exp=@n12)) {constB=0.25}
   B=LinearInterpolate(A=Add(A=@n21, B=@n23), B=MaterialFunctionCall() {materialFunction=&MF_SharedWinter}.SnowColor, Alpha=@n28)
  )
  n29 = ScalarParameter() {defaultValue=0.75, parameterName=NormalFlatness}
  n30 = Add(
   A=TextureCoordinate()
   B=ComponentMask(None=MaterialFunctionCall(Normal=@n18.RGB, Flatness=@n29) {materialFunction=&FlattenNormal}.Result) {g=true, r=true}
  )
  n31 = AppendVector(A=Constant(), B=ScalarParameter() {defaultValue=0.01, parameterName=ChromaOffset})
  n32 = Panner(Coordinate=Add(A=@n30, B=@n31)) {speedY=-0.05}
  n33 = Panner(Coordinate=@n30) {speedY=-0.05}
  n34 = Add(A=@n30, B=Multiply(A=@n31) {constB=-1})
  n35 = AppendVector(
   A=AppendVector(
    A=TextureSample(UVs=@n32) {texture=/Figment_S07/Effects/Fort_Effects/PlayerHealthDamage/T_ShieldCaustics}.R
    B=TextureSample(UVs=@n33) {texture=&T_ShieldCaustics}.G
   )
   B=TextureSample(UVs=Panner(Coordinate=@n34) {speedY=-0.05}) {texture=&T_ShieldCaustics}.B
  )
  n36 = Sine(None=ComponentMask(None=TextureCoordinate()) {g=true}) {period=2}
  n37 = Multiply(A=Add(A=MaterialFunctionCall() {materialFunction=&Color_Time_of_Day}.Result) {constB=0.25}) {constB=0.5}
  n38 = CollectionParameter() {
   collection=&FortniteMaterialParameters
   parameterName=SunAndMoonModelDirectionalVector
  }
  n39 = ComponentMask(None=@n38) {b=true, g=true, r=true}
  n40 = ReflectionVectorWS() {bNormalizeCustomWorldNormal=true}
  n41 = Clamp(None=Power(Base=DotProduct(A=@n39, B=@n40), Exp=ScalarParameter() {defaultValue=30, parameterName=Spec Highlight Power}) {constExponent=20})
  n42 = DotProduct(A=ComponentMask(None=Multiply(A=@n38) {constB=-1}) {b=true, g=true, r=true}, B=ReflectionVectorWS() {bNormalizeCustomWorldNormal=true})
  n43 = ScalarParameter() {defaultValue=50, parameterName=Reverse Spec Highlight Power}
  n44 = Multiply(A=@n37, B=Clamp(None=Power(Base=@n42, Exp=@n43) {constExponent=20}))
  n45 = ScalarParameter() {defaultValue=0.5, parameterName=ReverseHighlightOpacity}
  Add(
   A=Multiply(A=Multiply(A=@n35, B=@n36), B=ScalarParameter() {defaultValue=0.15, parameterName=Emissive Intensity})
   B=Add(A=Multiply(A=@n37, B=@n41), B=Multiply(A=@n44, B=@n45))
  )
  LinearInterpolate(B=ScalarParameter() {defaultValue=0.75, parameterName=Metallic}, Alpha=@n28) {constB=0}
  LinearInterpolate(Alpha=@n28) {constB=0.5}
  MaterialFunctionCall(Normal=@n18.RGB, Flatness=Constant() {r=0.75}) {materialFunction=&FlattenNormal}
 ]
}
