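# Translucent, refractive ice shell material for the outer Mooney bubble, capped with snow and frost.
# The ice layer combines panned scratch normals, chroma-offset shield caustics, sun-driven highlights and Fresnel opacity scaled by Fade In Sphere.
# A scratch mask shaped by a UV sine pushes vertices along their normals.
# An opaque snow layer, with flake specular on higher quality levels, blends over the ice through a frost-texture plus upward-facing mask.
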Material /Figment_S07/Athena/Prototype/Blueprints/Mooney/FX/M_Mooney_Bubble_Outer
graph {
 properties {
  bTangentSpaceNormal=false
  bUseMaterialAttributes=true
  blendMode=BLEND_Translucent
  ditherOpacityMask=true
  ditheredLODTransition=true
  pixelDepthOffsetMode=PDOM_Legacy
  refractionMethod=RM_IndexOfRefraction
  translucencyLightingMode=TLM_Surface
  usedShadingModels=Default Lit
 }
 nodes [
  n1 = ScalarParameter() {defaultValue=1, group=Smoke Colors, parameterName=Main Shading Fresnel Brightness}
  n2 = ScalarParameter() {defaultValue=2, group=Smoke Colors, parameterName=Main Shading Fresnel Exponent}
  n3 = CollectionParameter() {
   collection=&FortniteMaterialParameters
   parameterName=SunAndMoonModelDirectionalVector
  }
  n4 = ComponentMask(None=@n3) {b=true, g=true, r=true}
  n5 = TextureSample() {samplerType=SAMPLERTYPE_Masks, texture=/Game/Global/Textures/Default/Tileable_Patterns/T_IceScratch1_M}
  n6 = BumpOffset(Coordinate=TextureCoordinate(), Height=@n5.R, HeightRatioInput=Constant() {r=-0.05})
  n7 = Multiply(A=Time(), B=ScalarParameter() {defaultValue=0.25, parameterName=SpeedScale})
  n8 = MaterialFunctionCall(
   Normal=TextureSample(UVs=Panner(Coordinate=@n6, Time=@n7) {speedX=-0.05, speedY=-0.05}) {samplerType=SAMPLERTYPE_Normal, texture=/Figment_S07/Athena/WinterArt/Textures/T_IceScratch1_N}.RGB
   Flatness=ScalarParameter() {defaultValue=0.75, parameterName=NormalFlatness}
  ) {materialFunction=&FlattenNormal}
  n9 = Transform(None=@n8.Result)
  n10 = Add(A=TextureCoordinate(), B=ComponentMask(None=@n8.Result) {g=true, r=true})
  n11 = AppendVector(A=Constant(), B=ScalarParameter() {defaultValue=0.01, parameterName=ChromaOffset})
  n12 = Add(A=@n10, B=@n11)
  n13 = AppendVector(
   A=TextureSample(UVs=Panner(Coordinate=@n12) {speedY=-0.05}) {texture=/Figment_S07/Effects/Fort_Effects/PlayerHealthDamage/T_ShieldCaustics}.R
   B=TextureSample(UVs=Panner(Coordinate=@n10) {speedY=-0.05}) {texture=&T_ShieldCaustics}.G
  )
  n14 = Multiply(A=@n11) {constB=-1}
  n15 = TextureSample(UVs=Panner(Coordinate=Add(A=@n10, B=@n14)) {speedY=-0.05}) {texture=&T_ShieldCaustics}
  n16 = ComponentMask(None=TextureCoordinate()) {g=true}
  n17 = Multiply(
   A=Multiply(A=AppendVector(A=@n13, B=@n15.B), B=Sine(None=@n16) {period=2})
   B=ScalarParameter() {defaultValue=0.075, parameterName=Emissive Intensity}
  )
  n18 = Multiply(A=Add(A=MaterialFunctionCall() {materialFunction=&Color_Time_of_Day}.Result) {constB=0.25}) {constB=0.5}
  n19 = Multiply(A=Fresnel(ExponentIn=ScalarParameter() {defaultValue=9, parameterName=FresnelExpo}), B=@n18)
  n20 = CollectionParameter() {
   collection=&FortniteMaterialParameters
   parameterName=SunAndMoonModelDirectionalVector
  }
  n21 = DotProduct(A=ComponentMask(None=@n20) {b=true, g=true, r=true}, B=ReflectionVectorWS() {bNormalizeCustomWorldNormal=true})
  n22 = ScalarParameter() {defaultValue=30, parameterName=Spec Highlight Power}
  n23 = Multiply(A=@n18, B=Clamp(None=Power(Base=@n21, Exp=@n22) {constExponent=20}))
  n24 = Multiply(A=@n20) {constB=-1}
  n25 = Power(
   Base=DotProduct(A=ComponentMask(None=@n24) {b=true, g=true, r=true}, B=ReflectionVectorWS() {bNormalizeCustomWorldNormal=true})
   Exp=ScalarParameter() {defaultValue=50, parameterName=Reverse Spec Highlight Power}
  ) {constExponent=20}
  n26 = Multiply(A=Multiply(A=@n18, B=Clamp(None=@n25)), B=ScalarParameter() {defaultValue=0.5, parameterName=ReverseHighlightOpacity})
  n27 = Multiply(
   A=Fresnel(ExponentIn=ScalarParameter() {defaultValue=1.25, parameterName=Opacity FresnelExp}, Normal=@n9)
   B=LinearInterpolate(
    A=ScalarParameter() {defaultValue=3.5, parameterName=Fresnel Exp_Closed}
    B=ScalarParameter() {defaultValue=25, parameterName=Fresnel Exp_Open}
    Alpha=ScalarParameter() {defaultValue=1, parameterName=Fade In Sphere}
   )
  )
  n28 = OneMinus(None=Fresnel(Normal=@n9))
  n29 = Add(A=ComponentMask(None=@n8.Result) {r=true}, B=ComponentMask(None=@n8.Result) {g=true})
  n30 = Multiply(A=ScalarParameter() {defaultValue=25, parameterName=Scale}, B=Power(Base=@n5.R))
  n31 = Sine(None=ComponentMask(None=TextureCoordinate()) {g=true}) {period=2}
  n32 = SetMaterialAttributes!(
   MaterialAttributes
   Base Color=LinearInterpolate(
    A=VectorParameter() {parameterName=FlareIceColor, defaultValue=<rgba 0 0.7667 1 1>}.RGB
    B=MaterialFunctionCall() {materialFunction=/Game/Athena/WinterArt/Materials/MF_SharedWinter}.SnowColor
    Alpha=MaterialFunctionCall(Brightness=@n1, Exponent=@n2, Camera Vector=@n4, Vertex / Pixel Normal=@n9) {materialFunction=&MF_Fresnel_FX}.Fresnel
   )
   Emissive Color=Add(A=Add(A=@n17, B=@n19), B=Add(A=@n23, B=@n26))
   Opacity=@n27
   Normal=@n9
   Refraction=LinearInterpolate(Alpha=Multiply(A=@n27, B=@n28)) {constA=1, constB=1.02}
   Roughness=Constant()
   Specular=Constant() {r=0.25}
   Metallic=LinearInterpolate(B=ScalarParameter() {defaultValue=0.75, parameterName=Metallic}, Alpha=Saturate(None=@n29)) {constA=0.25}
   World Position Offset=Multiply(A=VertexNormalWS(), B=Multiply(A=@n30, B=@n31))
  ) {attributeSetTypes [
    #1
    #2
    #3
    #4
    #5
    #6
    #7
    #8
    #9
   ]}
  n33 = MaterialFunctionCall!(UseLandscapeCoordsOnSwitch) {materialFunction=/Game/Athena/Environments/Landscape/MaterialFunctions/Snow/MF_Athena_Snow_01}
  n34 = Constant() {r=0.2}
  n35 = Constant() {r=0.8}
  n36 = ComponentMask(None=MaterialFunctionCall(
    TextureObject=TextureObject() {samplerType=SAMPLERTYPE_LinearColor, texture=/Game/Athena/Environments/Landscape/Textures/T_Athena_Snow_01_M}
    TextureSize=ScalarParameter() {defaultValue=528.9877, parameterName=Specular Flake Tiling}
   ) {materialFunction=&WorldAlignedTexture}.XYZ Texture) {r=true}
  n37 = SetMaterialAttributes!(
   MaterialAttributes=@n33.Material
   Base Color=@n33.Material
   Roughness=Clamp(None=LinearInterpolate(A=@n34, B=@n35, Alpha=@n36))
   Specular=LinearInterpolate(Alpha=@n36) {constA=1, constB=0.35}
   Normal=@n33.Material
  ) {attributeSetTypes [
    #1
    #6
    #7
    #4
   ]}
  n38 = Constant() {r=1}
  n39 = SetMaterialAttributes!(
   MaterialAttributes=QualitySwitch(Default=@n37, Low=@n33.Material, High=@n37, Medium=@n33.Material)
   Normal
   Emissive Color=Constant()
   Refraction=@n38
   Opacity=@n38
  ) {attributeSetTypes [
    #4
    #2
    #5
    #3
   ]}
  n40 = AppendVector(A=ScalarParameter() {defaultValue=0.5, parameterName=Frost_U_Offset}, B=ScalarParameter() {defaultValue=0.5, parameterName=Frost_V_Offset})
  n41 = TextureCoordinate() {uTiling=2, vTiling=2}
  n42 = TextureCoordinate()
  n43 = Multiply(
   A=TextureSample(UVs=Add(A=@n40, B=@n41)) {texture=/Game/Effects/Fort_Effects/Textures/Snowflake/T_LightFrost}.G
   B=OneMinus(None=ComponentMask(None=@n42) {g=true})
  )
  n44 = Constant3Vector() {constant=<rgba 0 0 1 0>}
  n45 = Power(Base=DotProduct(A=Transform(None=@n44), B=@n44), Exp=ScalarParameter() {defaultValue=100, parameterName=Snow Contrast})
  MaterialFunctionCall!(Base Material=@n32, Top Material=@n39, Alpha=Saturate(None=Add(A=@n43, B=@n45))) {materialFunction=/Engine/Functions/MaterialLayerFunctions/MatLayerBlend_Standard}
 ]
}
