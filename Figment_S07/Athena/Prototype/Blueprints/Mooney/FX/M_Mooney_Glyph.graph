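# Masked, two-sided lit material for Mooney glyph effects, previewed on an ice tick mesh, that shows segments as a stepped ticker.
# Vertex color green is compared in Count steps against the Ticker parameter to switch segments on or off.
# A curve atlas gradient along UV V drives color and HDR glow, with time-scrolled noise varying the intensity.
# A camera-to-object sphere mask scales vertices out from the object position as distance grows.
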Material /Figment_S07/Athena/Prototype/Blueprints/Mooney/FX/M_Mooney_Glyph
graph {
 properties {
  bEnableResponsiveAA=true
  bTangentSpaceNormal=false
  blendMode=BLEND_Masked
  ditherOpacityMask=true
  pixelDepthOffsetMode=PDOM_Legacy
  previewMesh=&/S_FX_Mooney_IceTick
  translucencyLightingMode=TLM_VolumetricPerVertexDirectional
  twoSided=true
  usedShadingModels=Default Lit
 }
 nodes [
  n1 = ScalarParameter() {defaultValue=1, parameterName=Ticker}
  n2 = ScalarParameter() {defaultValue=12, parameterName=Count}
  n3 = Multiply(A=Add(A=OneMinus(None=@n1), B=Constant() {r=-0.0376}), B=@n2)
  Floor(None=@n3)
  n4 = SphereMask(
   A=ObjectPositionWS()
   B=CameraPositionWS()
   Radius=Multiply(A=ScalarParameter() {defaultValue=150000, parameterName=Camera Radius Range}, B=ScalarParameter() {defaultValue=3, parameterName=Ratio})
  ) {hardnessPercent=0}
  Multiply(
   A=Subtract(A=WorldPosition().XYZ, B=ObjectPositionWS())
   B=LinearInterpolate(B=ScalarParameter() {defaultValue=5, parameterName=Scale}, Alpha=OneMinus(None=@n4))
  )
  n5 = Multiply(A=VertexColor().G, B=@n2) {constB=4}
  n6 = Constant()
  Multiply(
   A=ScalarParameter() {defaultValue=1, parameterName=Opacity}
   B=If(A=Ceil(None=@n5), B=Ceil(None=@n3), "A > B"=Constant() {r=1}, "A == B"=@n6, "A < B"=@n6)
  )
  n7 = Power(Base=ComponentMask(None=TextureCoordinate()) {g=true}, Exp=ScalarParameter() {defaultValue=0.3616, parameterName=Color Contrast})
  n8 = CurveAtlasRowParameter(CurveTime=@n7) {atlas=&AthenaCurveAtlas01, curve=/Game/Athena/LinearColorCurves/IceFlamesGlyphs, defaultValue=49}
  n9 = ScalarParameter() {defaultValue=1, parameterName=HDR Intensity}
  n10 = Multiply(A=Time(), B=ScalarParameter() {defaultValue=-0.2, parameterName=NoisePannerSpeed})
  n11 = Multiply(
   A=Add(
    A=ScalarParameter() {defaultValue=0.5, parameterName=HDR Intensity_HIgh}
    B=TextureSample(UVs=@n10) {texture=/Figment_S07/Athena/Environments/DudeBro/T_Dudebro_Noise_01}.G
   )
   B=ScalarParameter() {defaultValue=2, parameterName=HIgh HDR Scale}
  )
  Multiply(
   A=MaterialFunctionCall(LDR Input=Clamp(None=@n8), HDR Tint=@n8, HDR Intensity Multiplier=LinearInterpolate(A=@n9, B=@n11, Alpha=@n7)) {materialFunction=&DeriveHDRfromLDR}.RemappedRange
   B=ScalarParameter() {defaultValue=0.35, parameterName=Emissive Scalar}
  ) {constB=0.1}
  Reroute(None=@n8)
  ScalarParameter() {defaultValue=0.75, parameterName=Metallic}
  ScalarParameter() {parameterName=Roughness}
 ]
}
