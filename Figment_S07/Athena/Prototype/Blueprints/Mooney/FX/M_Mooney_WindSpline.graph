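# Translucent, two-sided unlit material for wind-streak VFX on mesh particles, likely the Mooney wind spline, tinted by particle color and alpha.
# Two turbulence samples panned at different tilings and speeds multiply into a streak mask, shaped by a procedural UV edge falloff.
# View-angle, depth and distance-cull fades soften that mask.
# Time-of-day functions drive opacity and shadow color, which turbulence blends toward a bright highlight color.
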Material /Figment_S07/Athena/Prototype/Blueprints/Mooney/FX/M_Mooney_WindSpline
graph {
 properties {
  bUsedWithMeshParticles=true
  blendMode=BLEND_Translucent
  pixelDepthOffsetMode=PDOM_Legacy
  previewMesh=/Figment_S07/Effects/Fort_Effects/BluePrints/BeeHive/SM_Bee_FlyArc
  shadingModel=MSM_Unlit
  twoSided=true
  usedShadingModels=Unlit
 }
 nodes [
  n1 = ParticleColor()
  n2 = MaterialFunctionCall("FadeSensitivity (optional)"=ScalarParameter() {defaultValue=4, parameterName=AngleFade}) {materialFunction=&FadeBasedOnViewAngle}
  n3 = Multiply(A=Time(), B=ScalarParameter() {defaultValue=1, parameterName=MaskSpeed})
  n4 = TextureSample(UVs=Panner(Coordinate=TextureCoordinate() {uTiling=0.3, vTiling=3}, Time=@n3) {speedX=0.012, speedY=0.13}) {samplerType=SAMPLERTYPE_LinearColor, texture=/Game/Effects/Fort_Effects/Effects/Environment/Desert/Dust/Organize/T_Turbulence_Seamless}
  n5 = TextureSample(UVs=Panner(Coordinate=TextureCoordinate() {uTiling=0.6}, Time=@n3) {speedX=0.012, speedY=0.06}) {samplerType=SAMPLERTYPE_LinearColor, texture=&T_Turbulence_Seamless}
  n6 = ComponentMask(None=TextureCoordinate()) {g=true}
  n7 = ComponentMask(None=TextureCoordinate()) {r=true}
  n8 = Multiply(A=Multiply(A=OneMinus(None=@n6), B=@n6), B=Multiply(A=OneMinus(None=@n7), B=@n7))
  n9 = ScalarParameter() {defaultValue=1.35, parameterName=EXP}
  n10 = Multiply(
   A=Power(Base=Multiply(A=@n4.G, B=@n5.G), Exp=ScalarParameter() {defaultValue=1.15, parameterName=MaskPow})
   B=Multiply(A=Power(Base=@n8, Exp=@n9), B=ScalarParameter() {defaultValue=44, parameterName=ScaleAlphaMask})
  )
  n11 = Multiply(
   A=DistanceCullFade()
   B=DepthFade(Opacity=Multiply(A=@n2.Result, B=@n10), FadeDistance=ScalarParameter() {defaultValue=64, parameterName=FadeDistance})
  )
  n12 = ScalarParameter() {defaultValue=1, parameterName=Opacity_Morning}
  n13 = ScalarParameter() {defaultValue=1, parameterName=Opacity_Day}
  n14 = ScalarParameter() {defaultValue=1, parameterName=Opacity_Evening}
  n15 = ScalarParameter() {defaultValue=1, parameterName=Opacity_Night}
  Multiply(
   A=@n1.A
   B=MaterialFunctionCall(
    Color_Morning=Multiply(A=@n11, B=@n12)
    Color_Day=Multiply(A=@n11, B=@n13)
    Color_Evening=Multiply(A=@n11, B=@n14)
    Color_Night=Multiply(A=@n11, B=@n15)
   ) {materialFunction=&Color_Time_of_Day}.Result
  )
  n16 = VectorParameter() {parameterName=MorningShadowColor, defaultValue=<rgba 0.0976 0.4176 0.5776 1>}
  n17 = VectorParameter() {parameterName=DayShadowColor, defaultValue=<rgba 0.2789 0.5604 0.7011 1>}
  n18 = VectorParameter() {parameterName=EveningShadowColor, defaultValue=<rgba 0.3515 0.7091 0.8879 1>}
  n19 = VectorParameter() {parameterName=NightShadowColor, defaultValue=<rgba 0.1141 0.2855 0.3712 1>}
  n20 = TextureSample(UVs=TextureCoordinate() {uTiling=2}) {samplerType=SAMPLERTYPE_LinearColor, texture=&T_Turbulence_Seamless}
  Multiply(
   A=@n1.RGB
   B=LinearInterpolate(
    A=MaterialFunctionCall(Color_Morning=@n16.RGB, Color_Day=@n17.RGB, Color_Evening=@n18.RGB, Color_Night=@n19.RGB) {materialFunction=&Color_Time_of_Day}.Result
    B=VectorParameter() {parameterName=Color1, defaultValue=<rgba 0.188 0.797 1 1>}.RGB
    Alpha=@n20.G
   )
  )
 ]
}
