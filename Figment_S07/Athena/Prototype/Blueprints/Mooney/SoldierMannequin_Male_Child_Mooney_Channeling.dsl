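; A child Blueprint of a soldier mannequin skeletal mesh actor, named as the Mooney channeling variant, that adds an Ice King eye effect to the character's head.
; The construction script runs the parent's logic, then attaches the S_FX_IceKing_Eyes component to the "head" socket of the skeletal mesh.
; The attachment keeps relative location, rotation and scale, with welding of simulated bodies enabled.
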(fn ConstructionScript
  (|CallParentFunction)
  (Transformation|AttachComponentToComponent (Variables|Default|GetS_FX_IceKing_Eyes) (Variables|SkeletalMeshActor|GetSkeletalMeshComponent) "head" "KeepRelative" "KeepRelative" "KeepRelative" true))
