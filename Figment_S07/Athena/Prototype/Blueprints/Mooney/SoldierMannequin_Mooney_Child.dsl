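; SoldierMannequin_Mooney_Child is a child Blueprint, apparently of a Mooney soldier mannequin, that can switch off its attached body effect.
; A custom DisableFX event deactivates the P_Mooney_AttachedFX_Body component.
; The construction script only calls the parent's, and the BeginPlay, overlap and Tick events are empty.
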(event Custom|DisableFX
  (Components|Activation|SetActive (Variables|Default|GetP_Mooney_AttachedFX_Body) false false))

(event EventBeginPlay)

(event Collision|EventActorBeginOverlap
  (params (OtherActor :type Actor%20Object%20Reference)))

(event EventTick
  (params (DeltaSeconds :type Float%20%28single-precision%29)))

(fn ConstructionScript
  (|CallParentFunction))
